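# Prototype two-sided lit material for the Kiwi cube, with scalar switches blending purple, white, inert gray and blue states.
# A camera-facing ray-sphere term height-lerps noise masks into glow gradients, combined with three parallax rune layers and event-glow controls.
# Static-switched world-height sinking and pool gradients add warped-goo emissive, alongside an MPC-driven hand sphere mask.
# Base color gets parallax sparkles and edge tinting; quality switches simplify low-end shading.
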Material /CorruptionGameplay/Gameplay/Environmental/Cube/Prototype/PROTO_M_KiwiCube
graph {
 properties {bUseMaterialAttributes=true, pixelDepthOffsetMode=PDOM_Legacy, twoSided=true, usedShadingModels=Default Lit}
 nodes [
  n1 = WorldPosition()
  n2 = ScalarParameter() {defaultValue=-400, group=Switches, parameterName=Kiwi_SinkGrad_Offset}
  n3 = Add(A=@n1.XYZ, B=@n2)
  n4 = Constant() {r=-1}
  n5 = Saturate(None=Multiply(A=Multiply(A=@n3, B=@n4), B=ScalarParameter() {defaultValue=0.0009, group=SinkingGradient, parameterName=Kiwi_SinkGrad_Size}))
  n6 = NamedRerouteUsage()
  Add(A=@n5, B=@n6)
  Divide() {constB=5000}
  n7 = VectorParameter() {parameterName=Kiwi_Purple_GlowColor04, defaultValue=<rgba 0.3583 0 0.5 1>}
  n8 = Multiply(
   A=LinearInterpolate(A=Constant() {r=1200}, B=Constant() {r=1400}, Alpha=NamedRerouteUsage())
   B=ScalarParameter() {defaultValue=2, parameterName=CameraRadius_RadiusMult}
  )
  n9 = Multiply(A=@n8, B=@n8) {desc="r^2"}
  n10 = ObjectPositionWS()
  n11 = Add(A=VectorParameter() {parameterName=CameraRadius_SphereOffsetZ, defaultValue=<rgba 0 0 0 1>}.RGB, B=@n10)
  n12 = NamedRerouteUsage()
  n13 = Subtract(
   A=CameraPositionWS()
   B=Add(A=VectorParameter() {desc=For zapping players, parameterName=CameraRadius_SphereOffset}.RGB, B=LinearInterpolate(A=@n10, B=@n11, Alpha=@n12))
  )
  n14 = Distance(A=@n13, B=Constant())
  n15 = DotProduct(A=@n13, B=CameraVectorWS())
  n16 = Subtract(A=Multiply(A=@n14, B=@n14), B=Multiply(A=@n15, B=@n15))
  n17 = SquareRoot(None=Max(A=Subtract(A=@n9, B=@n16)) {constB=0})
  n18 = Multiply(
   A=Divide(A=Multiply(A=@n17) {constB=2}, B=Multiply(A=@n8) {constB=2})
   B=ScalarParameter() {defaultValue=0.87, parameterName=CameraRadius_DarkenAmount}
  )
  n19 = ScalarParameter() {defaultValue=0.7, parameterName=Cube_SphereMask_Radius}
  n20 = ScalarParameter() {defaultValue=0.7, parameterName=Cube_SphereMask_Radius High}
  n21 = NamedRerouteUsage()
  n22 = Power(
   Base=SphereMask(
    A=TextureCoordinate() {coordinateIndex=2}
    B=Constant() {r=0.5}
    Radius=LinearInterpolate(A=@n19, B=@n20, Alpha=@n21)
    Hardness=ScalarParameter() {parameterName=Cube_SphereMask_Hardness}
   ) {attenuationRadius=0.7, hardnessPercent=30}
   Exp=ScalarParameter() {defaultValue=1.5, parameterName=Cube_SphereMask_Exp}
  )
  n23 = Saturate(None=Add(A=NamedRerouteUsage(), B=ScalarParameter() {defaultValue=0.25, parameterName=Kiwi_MaskTexture_Soften}))
  n24 = ScalarParameter() {defaultValue=2.75, parameterName=Kiwi_MaskTexture_ShadowSinePowerMax}
  n25 = ScalarParameter() {defaultValue=1.5, parameterName=Kiwi_MaskTexture_ShadowSinePowerMin}
  n26 = NamedRerouteUsage()
  n27 = Multiply(A=@n23, B=Power(Base=NamedRerouteUsage(), Exp=LinearInterpolate(A=@n24, B=@n25, Alpha=@n26)))
  n28 = QualitySwitch(Default=@n27, Low=@n23, High=@n27)
  n29 = MaterialFunctionCall(
   Transition Phase=@n18
   Height Texture=Multiply(A=NamedRerouteDeclaration(None=@n22) {name=SphereMask, nodeColor=<rgba 1 0.2578 0 1>}, B=@n28)
  ) {materialFunction=&HeightLerp}
  n30 = LinearInterpolate(
   A=MaterialFunctionCall(
    1=VectorParameter() {parameterName=Kiwi_Purple_GlowColor01, defaultValue=<rgba 0.012 0.0011 0.08 1>}.RGB
    2=VectorParameter() {parameterName=Kiwi_Purple_GlowColor02, defaultValue=<rgba 0.0356 0.001 0.2 1>}.RGB
    3=VectorParameter() {parameterName=Kiwi_Purple_GlowColor03, defaultValue=<rgba 0.1264 0 0.3 1>}.RGB
    4=@n7.RGB
    A=@n29.Lerp Alpha No Contrast
   ) {materialFunction=&Lerp_Multiple_Float3}.Lerp 4 Inputs
   B=MaterialFunctionCall(
    1=VectorParameter() {parameterName=Kiwi_WhiteTransition_GlowColor01, defaultValue=<rgba 2 0.7028 0 1>}.RGB
    2=VectorParameter() {parameterName=Kiwi_WhiteTransition_GlowColor02, defaultValue=<rgba 1 0.4207 0.2 1>}.RGB
    3=VectorParameter() {parameterName=Kiwi_WhiteTransition_GlowColor03, defaultValue=<rgba 0.8 0.7031 0.6792 1>}.RGB
    4=VectorParameter() {parameterName=Kiwi_WhiteTransition_GlowColor04, defaultValue=<rgba 1 1 1 1>}.RGB
    A=@n29.Lerp Alpha No Contrast
   ) {materialFunction=&Lerp_Multiple_Float3}.Lerp 4 Inputs
   Alpha=NamedRerouteUsage()
  )
  n31 = MaterialFunctionCall(
   1=VectorParameter() {parameterName=Kiwi_White_GlowColor01, defaultValue=<rgba 0.1 0.0811 0.06 1>}.RGB
   2=VectorParameter() {parameterName=Kiwi_White_GlowColor02, defaultValue=<rgba 0.2 0.1811 0.16 1>}.RGB
   3=VectorParameter() {parameterName=Kiwi_White_GlowColor03, defaultValue=<rgba 0.3 0.2858 0.27 1>}.RGB
   4=VectorParameter() {parameterName=Kiwi_White_GlowColor04, defaultValue=<rgba 0.5 0.5 0.5 1>}.RGB
   A=@n29.Lerp Alpha No Contrast
  ) {materialFunction=&Lerp_Multiple_Float3}
  n32 = NamedRerouteUsage()
  n33 = VectorParameter() {group=inert Grey Options, parameterName=Kiwi_InertGray_GlowColor01, defaultValue=<rgba 0.0417 0.0417 0.0417 1>}
  n34 = VectorParameter() {group=inert Grey Options, parameterName=Kiwi_InertGray_GlowColor02, defaultValue=<rgba 0.1358 0.146 0.1696 1>}
  n35 = VectorParameter() {group=inert Grey Options, parameterName=Kiwi_InertGray_GlowColor03, defaultValue=<rgba 0.2482 0.2682 0.3 1>}
  n36 = VectorParameter() {group=inert Grey Options, parameterName=Kiwi_InertGray_GlowColor04, defaultValue=<rgba 0.4062 0.4382 0.5 1>}
  n37 = MaterialFunctionCall(1=@n33.RGB, 2=@n34.RGB, 3=@n35.RGB, 4=@n36.RGB, A=@n29.Lerp Alpha No Contrast) {materialFunction=&Lerp_Multiple_Float3}
  n38 = MaterialFunctionCall(Value=Multiply(A=Time(), B=ScalarParameter() {defaultValue=0.5, parameterName=Kiwi_PulseSpeed})) {materialFunction=&LinearSine}
  n39 = LinearInterpolate(
   A=ScalarParameter() {defaultValue=1, group=inert Grey Options, parameterName=Kiwi_InertGray Mul Low}
   B=ScalarParameter() {defaultValue=1, group=inert Grey Options, parameterName=Kiwi_InertGray Mul high}
   Alpha=@n38."Rounded Linear Sine "
  )
  n40 = LinearInterpolate(
   A=LinearInterpolate(A=LinearInterpolate(A=@n30, B=@n31.Lerp 4 Inputs, Alpha=@n32), B=Multiply(A=@n37.Lerp 4 Inputs, B=@n39), Alpha=NamedRerouteUsage())
   B=MaterialFunctionCall(
    1=VectorParameter() {parameterName=Kiwi_Blue_GlowColor01, defaultValue=<rgba 0.01 0 0.05 1>}.RGB
    2=VectorParameter() {parameterName=Kiwi_Blue_GlowColor02, defaultValue=<rgba 0.0122 0 0.1 1>}.RGB
    3=VectorParameter() {parameterName=Kiwi_Blue_GlowColor03, defaultValue=<rgba 0.006 0.0666 0.3 1>}.RGB
    4=VectorParameter() {parameterName=Kiwi_Blue_GlowColor04, defaultValue=<rgba 0.0074 0.3685 0.6 1>}.RGB
    A=@n29.Lerp Alpha No Contrast
   ) {materialFunction=&Lerp_Multiple_Float3}.Lerp 4 Inputs
   Alpha=NamedRerouteUsage()
  )
  LinearInterpolate(A=VectorParameter() {parameterName=Kiwi_Blue_GlowColorEdge, defaultValue=<rgba 0 0.7729 1 1>}.RGB, B=@n40)
  n41 = Multiply(
   A=Power(Base=@n18, Exp=ScalarParameter() {defaultValue=20, parameterName=CameraRadius_PowerAmount_GradientControl})
   B=ScalarParameter() {defaultValue=1, parameterName=CameraRadius_ActiveHotspot}
  )
  n42 = NamedRerouteUsage()
  n43 = Multiply(A=@n41, B=@n42)
  n44 = Multiply(A=QualitySwitch(Default=@n43, Low=@n41, High=@n43), B=NamedRerouteUsage())
  n45 = Multiply(A=Add(A=@n40, B=Multiply(A=@n7.RGB, B=@n44)), B=NamedRerouteUsage())
  n46 = ScalarParameter() {defaultValue=0.25, parameterName=Kiwi_MaskTexture_ShadowDeactivated}
  n47 = ScalarParameter() {defaultValue=1, parameterName=Kiwi_MaskTexture_ShadowActivated}
  n48 = NamedRerouteUsage()
  n49 = Saturate(None=Add(A=LinearInterpolate(A=@n46, B=@n47, Alpha=@n48), B=@n42))
  n50 = Power(
   Base=Multiply(A=NamedRerouteUsage(), B=@n29.Lerp Alpha No Contrast)
   Exp=ScalarParameter() {defaultValue=2.5, parameterName=Kiwi_MaskTexture_Power}
  )
  n51 = ScalarParameter() {defaultValue=-0.1, parameterName=Kiwi_Runes_deactivated}
  n52 = ScalarParameter() {defaultValue=1, parameterName=Kiwi_Runes_activated}
  n53 = NamedRerouteUsage()
  n54 = VectorParameter() {parameterName=Kiwi_Runes_BaseColorWhite, defaultValue=<rgba 4 3.4718 3.1206 1>}
  n55 = LinearInterpolate(A=VectorParameter() {parameterName=Kiwi_Runes_BaseColor, defaultValue=<rgba 0.6076 0 1 1>}.RGB, B=@n54.RGB, Alpha=NamedRerouteUsage())
  n56 = NamedRerouteUsage()
  n57 = VectorParameter() {parameterName=Kiwi_Runes_BaseColorBlue, defaultValue=<rgba 0 0.7909 1 1>}
  n58 = LinearInterpolate(
   A=LinearInterpolate(A=LinearInterpolate(A=@n55, B=@n54.RGB, Alpha=@n56), B=@n57.RGB, Alpha=NamedRerouteUsage())
   B=@n57.RGB
   Alpha=NamedRerouteUsage()
  )
  n59 = MaterialFunctionCall(Float2=Multiply(A=ScalarParameter() {defaultValue=75, parameterName=Kiwi_Runes_SmallTileRate}, B=TextureCoordinate() {uTiling=0.95, vTiling=0.75})) {materialFunction=&BreakOutFloat2Components}
  n60 = MaterialFunctionCall(
   X=Add(A=ScalarParameter() {parameterName=Kiwi_Runes_SmallOffsetX}, B=@n59.R)
   Y=Add(A=@n59.G, B=ScalarParameter() {defaultValue=0.5, parameterName=Kiwi_Runes_SmallOffsetY})
  ) {materialFunction=&MakeFloat2}
  n61 = Constant() {r=-1}
  n62 = Constant() {r=1}
  n63 = TextureSample(UVs=@n60.Result) {
   samplerSource=SSM_Wrap_WorldGroupSettings
   samplerType=SAMPLERTYPE_LinearColor
   texture=/Game/Athena/Prototype/Blueprints/Cube/Materials/DFHeightMask
  }
  n64 = TextureSample(UVs=@n60.Result) {
   samplerSource=SSM_Wrap_WorldGroupSettings
   samplerType=SAMPLERTYPE_LinearColor
   texture=/Game/Athena/Prototype/Blueprints/Cube/Materials/T_Cube_Rune_Blanks
  }
  n65 = LinearInterpolate(A=@n63.G, B=@n64.G, Alpha=ScalarParameter() {parameterName=Use Blank Runes})
  n66 = BumpOffset(
   Coordinate=@n60.Result
   Height=Multiply(A=LinearInterpolate(A=@n61, B=@n62, Alpha=@n65) {constA=-1}, B=ScalarParameter() {defaultValue=1, parameterName=Kiwi_Runes_SmallContrast})
   HeightRatioInput=ScalarParameter() {defaultValue=0.06, parameterName=Kiwi_Runes_SmallBumpOffsetHeight}
  )
  n67 = TextureSample(UVs=@n66) {
   samplerSource=SSM_Wrap_WorldGroupSettings
   samplerType=SAMPLERTYPE_LinearColor
   texture=&DFHeightMask
  }
  n68 = TextureSample(UVs=@n66) {
   samplerSource=SSM_Wrap_WorldGroupSettings
   samplerType=SAMPLERTYPE_LinearColor
   texture=&T_Cube_Rune_Blanks
  }
  n69 = LinearInterpolate(A=@n67.G, B=@n68.G, Alpha=ScalarParameter() {parameterName=Use Blank Runes})
  n70 = LinearInterpolate(
   A=ScalarParameter() {defaultValue=10, parameterName=Kiwi_Runes_SmallRangeHi}
   B=ScalarParameter() {defaultValue=2, parameterName=Kiwi_Runes_SmallRangeLow}
   Alpha=NamedRerouteUsage()
  )
  n71 = MaterialFunctionCall(Float2=Multiply(A=ScalarParameter() {defaultValue=19, parameterName=Kiwi_Runes_LargeTileRate}, B=TextureCoordinate() {vTiling=0.75})) {materialFunction=&BreakOutFloat2Components}
  n72 = MaterialFunctionCall(
   X=Add(A=ScalarParameter() {defaultValue=0.1, parameterName=Kiwi_Runes_LargeOffsetX}, B=@n71.R)
   Y=Add(A=@n71.G, B=ScalarParameter() {defaultValue=0.05, parameterName=Kiwi_Runes_LargeOffsetY})
  ) {materialFunction=&MakeFloat2}
  n73 = Constant() {r=-1}
  n74 = Constant() {r=1}
  n75 = TextureSample(UVs=@n72.Result) {
   samplerSource=SSM_Wrap_WorldGroupSettings
   samplerType=SAMPLERTYPE_LinearColor
   texture=&DFHeightMask
  }
  n76 = TextureSample(UVs=@n72.Result) {
   samplerSource=SSM_Wrap_WorldGroupSettings
   samplerType=SAMPLERTYPE_LinearColor
   texture=&T_Cube_Rune_Blanks
  }
  n77 = LinearInterpolate(A=@n75.G, B=@n76.G, Alpha=ScalarParameter() {parameterName=Use Blank Runes})
  n78 = BumpOffset(
   Coordinate=@n72.Result
   Height=Multiply(A=LinearInterpolate(A=@n73, B=@n74, Alpha=@n77) {constA=-1}, B=ScalarParameter() {defaultValue=0.5, parameterName=Kiwi_Runes_LargeContrast})
   HeightRatioInput=ScalarParameter() {defaultValue=0.15, parameterName=Kiwi_Runes_LargeBumpOffsetHeight}
  )
  n79 = TextureSample(UVs=@n78) {
   samplerSource=SSM_Wrap_WorldGroupSettings
   samplerType=SAMPLERTYPE_LinearColor
   texture=&DFHeightMask
  }
  n80 = TextureSample(UVs=@n78) {
   samplerSource=SSM_Wrap_WorldGroupSettings
   samplerType=SAMPLERTYPE_LinearColor
   texture=&T_Cube_Rune_Blanks
  }
  n81 = LinearInterpolate(A=@n79.G, B=@n80.G, Alpha=ScalarParameter() {parameterName=Use Blank Runes})
  n82 = ScalarParameter() {defaultValue=20, parameterName=Kiwi_Runes_LargeGradientFalloff01}
  n83 = Multiply(A=@n81, B=ScalarParameter() {defaultValue=1, parameterName=Kiwi_Runes_LargeMaskDistance})
  n84 = Multiply(A=ScalarParameter() {defaultValue=32, parameterName=Kiwi_Runes_MediumTileRate}, B=TextureCoordinate())
  n85 = TextureSample(UVs=@n84) {
   samplerSource=SSM_Wrap_WorldGroupSettings
   samplerType=SAMPLERTYPE_LinearColor
   texture=&DFHeightMask
  }
  n86 = TextureSample(UVs=@n84) {
   samplerSource=SSM_Wrap_WorldGroupSettings
   samplerType=SAMPLERTYPE_LinearColor
   texture=&T_Cube_Rune_Blanks
  }
  n87 = ScalarParameter() {parameterName=Use Blank Runes}
  n88 = BumpOffset(
   Coordinate=@n84
   Height=LinearInterpolate(
    A=ScalarParameter() {defaultValue=-1, parameterName=Kiwi_Runes_MediumWarpContrastLow}
    B=ScalarParameter() {defaultValue=1, parameterName=Kiwi_Runes_MediumWarpContrastHi}
    Alpha=LinearInterpolate(A=@n85.R, B=@n86.R, Alpha=@n87)
   )
   HeightRatioInput=ScalarParameter() {defaultValue=0.06, parameterName=Kiwi_Runes_MediumBumpOffsetHeight}
  )
  n89 = QualitySwitch(Default=@n88, Low=@n84, High=@n88)
  n90 = TextureSample(UVs=@n89) {
   samplerSource=SSM_Wrap_WorldGroupSettings
   samplerType=SAMPLERTYPE_LinearColor
   texture=&DFHeightMask
  }
  n91 = TextureSample(UVs=@n89) {
   samplerSource=SSM_Wrap_WorldGroupSettings
   samplerType=SAMPLERTYPE_LinearColor
   texture=&T_Cube_Rune_Blanks
  }
  n92 = LinearInterpolate(A=@n90.R, B=@n91.R, Alpha=ScalarParameter() {parameterName=Use Blank Runes})
  n93 = Panner(
   Coordinate=Multiply(A=ScalarParameter() {defaultValue=2, parameterName=Kiwi_Runes_MediumMaskTileRate}, B=TextureCoordinate())
   Time=Multiply(A=ScalarParameter() {defaultValue=0.0135, parameterName=Kiwi_Runes_MediumMaskSpeed}, B=Time())
  ) {speedX=1}
  n94 = TextureSample(UVs=@n93) {
   samplerSource=SSM_Wrap_WorldGroupSettings
   samplerType=SAMPLERTYPE_LinearColor
   texture=&DFHeightMask
  }
  n95 = TextureSample(UVs=@n93) {
   samplerSource=SSM_Wrap_WorldGroupSettings
   samplerType=SAMPLERTYPE_LinearColor
   texture=&T_Cube_Rune_Blanks
  }
  n96 = ScalarParameter() {parameterName=Use Blank Runes}
  n97 = Saturate(None=LinearInterpolate(
    A=ScalarParameter() {defaultValue=1, parameterName=Kiwi_Runes_MediumMaskContrastHi}
    B=ScalarParameter() {defaultValue=-1, parameterName=Kiwi_Runes_MediumMaskContrastLow}
    Alpha=LinearInterpolate(A=@n94.B, B=@n95.B, Alpha=@n96)
   ))
  n98 = Saturate(None=Multiply(A=Multiply(A=@n92, B=@n97), B=ScalarParameter() {defaultValue=1.25, parameterName=Kiwi_Runes_MediumMaskDistance}))
  n99 = ScalarParameter() {defaultValue=1.5, parameterName=Kiwi_Runes_MediumContrastLow}
  n100 = ScalarParameter() {defaultValue=15, parameterName=Kiwi_Runes_MediumContrastHi}
  n101 = NamedRerouteUsage()
  n102 = Multiply(
   A=Power(Base=@n92, Exp=LinearInterpolate(A=@n99, B=@n100, Alpha=@n101))
   B=ScalarParameter() {defaultValue=5, parameterName=Kiwi_Runes_MediumIntensity}
  )
  n103 = LinearInterpolate(A=LinearInterpolate(A=Power(Base=@n69, Exp=@n70), B=Power(Base=@n81, Exp=@n82), Alpha=Saturate(None=@n83)), B=@n102, Alpha=@n98)
  n104 = Multiply(
   A=ScalarParameter() {defaultValue=1, desc="0=Off 1=On", group=Switches, parameterName=Kiwi_Runes_On}
   B=Multiply(A=LinearInterpolate(A=@n51, B=@n52, Alpha=@n53), B=Multiply(A=@n58, B=@n103))
  )
  n105 = Add(A=Multiply(A=@n45, B=@n49), B=Multiply(A=@n50, B=Multiply(A=@n49, B=@n104)))
  n106 = ScalarParameter() {desc=Center Glow for pusling effects, group=Switches, parameterName=Kiwi_Runes_Event_Glow01}
  n107 = ScalarParameter() {desc=Center Glow for pusling effects, group=Switches, parameterName=Kiwi_Runes_Event_Glow02}
  n108 = ScalarParameter() {desc=Bright pulsing center glow for pusling effects, group=Switches, parameterName=Kiwi_Runes_Event_PulsedBrightRunes}
  n109 = Add(A=Add(A=Multiply(A=@n106) {constB=2}, B=Multiply(A=@n107) {constB=1.5}), B=@n108)
  n110 = LinearInterpolate(
   A=Multiply(A=ScalarParameter() {defaultValue=0.18, parameterName=Kiwi_Runes_Event_AmbiantGlow}, B=@n105)
   B=Multiply(A=@n105, B=Constant() {r=1})
   Alpha=@n109
  )
  n111 = ScalarParameter() {defaultValue=0.01, parameterName=Kiwi_Runes_Event_AmbiantGlow_forMobile}
  n112 = Multiply(A=@n102, B=@n98)
  n113 = ScalarParameter() {defaultValue=1, desc="0=Off 1=On", group=Switches, parameterName=Kiwi_Runes_On}
  n114 = Add(A=@n45, B=Multiply(A=@n45, B=Multiply(A=@n112, B=@n113)))
  n115 = Constant() {r=1}
  n116 = QualitySwitch(Default=@n110, Low=LinearInterpolate(A=Multiply(A=@n111, B=@n114), B=Multiply(A=@n114, B=@n115), Alpha=@n109), High=@n110)
  LinearInterpolate(A=@n116, B=Power(Base=@n116), Alpha=NamedRerouteUsage())
  n117 = ScalarParameter() {defaultValue=15, group=SinkingGradient, parameterName=Kiwi_SinkGrad_Power}
  LinearInterpolate(A=@n117, B=ScalarParameter() {defaultValue=1, group=SinkingGradient, parameterName=Kiwi_SinkGrad_Power Hand}, Alpha=NamedRerouteUsage())
  MaterialFunctionCall(Number=@n38."Rounded Linear Sine ") {materialFunction=&DebugScalarValues}
  n118 = TwoSidedSign()
  n119 = Multiply(A=NamedRerouteUsage(), B=Constant() {r=2})
  n120 = NamedRerouteUsage()
  n121 = MaterialFunctionCall(Transition Phase=Power(Base=Saturate(None=@n119), Exp=Constant() {r=1.8}), Height Texture=@n120) {materialFunction=&HeightLerp}
  n122 = LinearInterpolate(
   A=ScalarParameter() {defaultValue=0.2, parameterName=Specular_Low}
   B=ScalarParameter() {defaultValue=0.1, parameterName=Specular_High}
   Alpha=Saturate(None=Add(A=@n121.Lerp Alpha No Contrast, B=@n120))
  )
  Multiply(A=@n118, B=@n122)
  n123 = Saturate(None=Multiply(A=NamedRerouteUsage(), B=Constant() {r=1.4}))
  n124 = Constant() {r=1.5}
  n125 = LinearInterpolate(
   A=ScalarParameter() {defaultValue=0.29, parameterName=Roughness_Low}
   B=ScalarParameter() {defaultValue=0.8, parameterName=Roughness_High}
   Alpha=MaterialFunctionCall(Transition Phase=Power(Base=@n123, Exp=@n124), Height Texture=NamedRerouteUsage()) {materialFunction=&HeightLerp}.Lerp Alpha No Contrast
  )
  Multiply(A=@n118, B=@n125)
  n126 = LinearInterpolate(
   A=VectorParameter() {parameterName=Kiwi_Purple_BaseColor, defaultValue=<rgba 0.015 0.0013 0.1 1>}.RGB
   B=VectorParameter() {parameterName=Kiwi_WhiteTransition_BaseColor, defaultValue=<rgba 0.5 0.3865 0.325 1>}.RGB
   Alpha=NamedRerouteUsage()
  )
  n127 = VectorParameter() {parameterName=Kiwi_White_BaseColor, defaultValue=<rgba 0.5 0.4571 0.4338 1>}
  n128 = NamedRerouteUsage()
  n129 = VectorParameter() {group=inert Grey Options, parameterName=Kiwi_InertGray_GlowColor01, defaultValue=<rgba 0.0417 0.0417 0.0417 1>}
  n130 = LinearInterpolate(
   A=LinearInterpolate(A=LinearInterpolate(A=@n126, B=@n127.RGB, Alpha=@n128), B=@n129.RGB, Alpha=NamedRerouteUsage())
   B=VectorParameter() {parameterName=Kiwi_Blue_BaseColor, defaultValue=<rgba 0.0116 0.1478 0.5 1>}.RGB
   Alpha=NamedRerouteUsage()
  )
  n131 = VectorParameter() {group=Kiwi_Goo, parameterName=Kiwi_Sparkles_ColorWhite, defaultValue=<rgba 2 1.8358 1.3393 1>}
  n132 = LinearInterpolate(
   A=LinearInterpolate(
    A=VectorParameter() {group=Kiwi_Goo, parameterName=Kiwi_Sparkles_Color, defaultValue=<rgba 0.7322 0.25 1.25 1>}.RGB
    B=@n131.RGB
    Alpha=NamedRerouteUsage()
   )
   B=@n131.RGB
   Alpha=NamedRerouteUsage()
  )
  n133 = NamedRerouteUsage()
  n134 = Constant() {r=14}
  n135 = TextureCoordinate()
  n136 = Constant() {r=-2}
  n137 = TextureSample(UVs=BumpOffset(Coordinate=Multiply(A=@n134, B=@n135), Height=@n136, HeightRatioInput=Constant() {r=0.02})) {samplerSource=SSM_Wrap_WorldGroupSettings, samplerType=SAMPLERTYPE_Masks, texture=/Game/Global/Textures/Default/Noise/T_CubeNoise}
  n138 = Constant() {r=28}
  n139 = TextureSample(UVs=BumpOffset(Coordinate=Multiply(A=@n135, B=@n138), Height=@n136, HeightRatioInput=Constant() {r=0.2})) {samplerSource=SSM_Wrap_WorldGroupSettings, samplerType=SAMPLERTYPE_Masks, texture=&T_CubeNoise}
  n140 = Multiply(A=Power(Base=Multiply(A=@n137.B, B=@n139.B), Exp=Constant() {r=7}), B=Constant() {r=60})
  n141 = Multiply(
   A=LinearInterpolate(
    A=LinearInterpolate(A=@n132, Alpha=@n133) {constB=0}
    B=VectorParameter() {group=Kiwi_Goo, parameterName=Kiwi_Sparkles_ColorBlue, defaultValue=<rgba 0.5915 0.8915 1.25 1>}.RGB
    Alpha=NamedRerouteUsage()
   )
   B=@n140
  )
  n142 = Multiply(A=Constant() {r=3.5}, B=TextureCoordinate())
  n143 = TextureSample!(
   UVs=BumpOffset(Coordinate=@n142, Height=Constant() {r=0.58}, HeightRatioInput=Constant() {r=-0.5})
   Tex
   Bias=ScalarParameter() {defaultValue=5, parameterName=mipbias1}
   Apply View MipBias
  ) {
   mipValueMode=TMVM_MipBias
   samplerSource=SSM_Wrap_WorldGroupSettings
   samplerType=SAMPLERTYPE_Masks
   texture=&T_CubeNoise
  }
  n144 = Constant() {r=3}
  n145 = Saturate(None=Add(A=Power(Base=@n143.R, Exp=@n144), B=Constant() {r=-0.05}))
  n146 = OneMinus(None=NamedRerouteUsage())
  n147 = Constant() {r=1.7}
  n148 = Power(Base=Saturate(None=Multiply(A=@n146, B=@n147)), Exp=Constant() {r=30})
  n149 = Add(A=@n130, B=Multiply(A=Multiply(A=@n141, B=@n145), B=@n148))
  n150 = LinearInterpolate(
   A=LinearInterpolate(
    A=VectorParameter() {parameterName=Kiwi_Purple_EdgeBaseColor, defaultValue=<rgba 0.006 0.0005 0.04 1>}.RGB
    B=VectorParameter() {parameterName=Kiwi_WhiteTransition_EdgeBaseColor, defaultValue=<rgba 0.75 0.6441 0.525 1>}.RGB
    Alpha=NamedRerouteUsage()
   )
   B=VectorParameter() {parameterName=Kiwi_White_EdgeBaseColor, defaultValue=<rgba 0.35 0.3199 0.3037 1>}.RGB
   Alpha=NamedRerouteUsage()
  )
  n151 = VectorParameter() {group=inert Grey Options, parameterName=Kiwi_InertGray_GlowColor02, defaultValue=<rgba 0.1358 0.146 0.1696 1>}
  n152 = NamedRerouteUsage()
  n153 = Saturate(None=Multiply(A=NamedRerouteUsage(), B=Constant() {r=1.95}))
  n154 = Constant() {r=5}
  n155 = LinearInterpolate(
   A=QualitySwitch(Default=@n149, Low=@n130, High=@n149)
   B=LinearInterpolate(
    A=LinearInterpolate(A=@n150, B=@n151.RGB, Alpha=@n152)
    B=VectorParameter() {parameterName=Kiwi_Blue_EdgeBaseColor, defaultValue=<rgba 0.16 0.869 1 1>}.RGB
    Alpha=NamedRerouteUsage()
   )
   Alpha=MaterialFunctionCall(Transition Phase=Power(Base=@n153, Exp=@n154), Height Texture=NamedRerouteUsage()) {materialFunction=&HeightLerp}.Lerp Alpha No Contrast
  )
  Multiply(A=@n155, B=@n118)
  n156 = NamedRerouteUsage()
  n157 = Multiply(A=Multiply(A=NamedRerouteUsage()) {constB=1.25}, B=Add(A=NamedRerouteUsage(), B=Add(A=@n156, B=@n156)))
  n158 = WorldPosition()
  n159 = Subtract(A=NamedRerouteUsage(), B=ScalarParameter() {defaultValue=0.25, group=Hand_SphereMask, parameterName=Large Noise Offset modded}) {constB=0.25}
  n160 = ScalarParameter() {group=Hand_SphereMask, parameterName=Use Debug}
  n161 = LinearInterpolate(
   A=CollectionParameter() {collection=&/PROTO_MPC_Kiwi, parameterName=Kiwi_KevinHand_SphereRadius}
   B=ScalarParameter() {defaultValue=250, group=Hand_SphereMask, parameterName=Debug Rad}
   Alpha=@n160
  )
  n162 = LinearInterpolate(
   A=ScalarParameter() {defaultValue=2, group=Hand_SphereMask, parameterName=Large Noise Str modded Low}
   B=ScalarParameter() {defaultValue=1, group=Hand_SphereMask, parameterName=Large Noise Str modded High}
   Alpha=@n161
  )
  n163 = CollectionParameter() {collection=&PROTO_MPC_Kiwi, parameterName=Kiwi_KevinHand_SphereMask_Location}
  n164 = LinearInterpolate(
   A=ComponentMask(None=@n163) {b=true, g=true, r=true}
   B=Add(
    A=MaterialFunctionCall() {desc=Debug, materialFunction=&ComponentPivotLocation}.Result
    B=VectorParameter() {group=Hand_SphereMask, parameterName=Debug pos, defaultValue=<rgba 768 0 -600 1>}.RGB
   )
   Alpha=@n160
  )
  n165 = LinearInterpolate(
   A=Add(A=@n158.XYZ, B=Multiply(A=@n159, B=@n162) {constB=500})
   B=@n158.XYZ
   Alpha=SphereMask(A=@n158.XYZ, B=@n164, Radius=@n161, Hardness=ScalarParameter() {group=Hand_SphereMask, parameterName=Hand distortion sphere mask hardness}) {hardnessPercent=25}
  )
  n166 = InverseLinearInterpolate(Value=@n161) {bClampResult=true, constB=5000}
  n167 = LinearInterpolate(
   A=ScalarParameter() {defaultValue=0.5, group=Hand_SphereMask, parameterName=Hand_SphereMask_Hardness Low}
   B=ScalarParameter() {defaultValue=1, group=Hand_SphereMask, parameterName=Hand_SphereMask_Hardness High}
   Alpha=@n166
  )
  n168 = ScalarParameter() {defaultValue=1, group=Hand_SphereMask, parameterName=Hand_SphereMask_Exponent Low}
  n169 = ScalarParameter() {defaultValue=1, group=Hand_SphereMask, parameterName=Hand_SphereMask_Exponent high}
  n170 = Multiply(
   A=Power(Base=SphereMask(A=@n165, B=@n164, Radius=@n161, Hardness=@n167) {hardnessPercent=0}, Exp=LinearInterpolate(A=@n168, B=@n169, Alpha=@n166))
   B=LinearInterpolate(
    A=ScalarParameter() {defaultValue=0.5, group=Hand_SphereMask, parameterName=Hand_SphereMask_Multiply low}
    B=ScalarParameter() {defaultValue=1, group=Hand_SphereMask, parameterName=Hand_SphereMask_Multiply High}
    Alpha=@n166
   ) {constA=1}
  )
  Multiply(
   A=Add(
    A=VectorParameter() {group=SinkingGradient, parameterName=Kiwi_SinkGrad_ColorBlue, defaultValue=<rgba 0 0.5694 1.25 1>}.RGB
    B=Multiply(A=@n157) {constB=0.4}
   )
   B=@n170
  )
  n171 = MaterialFunctionCall!(UseLocalSpace=StaticBool() {value=true}, WorldPosition, NormalOverride) {materialFunction=/Game/Environments/Apollo/Sets/Skyfire/Materials/MaterialFunctions/MF_WorldAlignedUVs_HardEdge}
  n172 = ScalarParameter() {defaultValue=500, parameterName=Kiwi_Wave_WarpTilerate}
  n173 = ScalarParameter() {defaultValue=200, group=SinkingGradient, parameterName=Kiwi_SinkGrad_GooWarpTilerate, sortPriority=50}
  n174 = NamedRerouteUsage()
  n175 = NamedRerouteUsage()
  n176 = Divide(A=@n171.Result, B=LinearInterpolate(A=LinearInterpolate(A=@n172, B=@n173, Alpha=@n174), B=@n173, Alpha=@n175))
  n177 = Multiply(A=Time(), B=ScalarParameter() {defaultValue=0.4, parameterName=Kiwi_Wave_WarpSpeed})
  n178 = ScalarParameter() {defaultValue=1.5, parameterName=Kiwi_Wave_WarpNoiseIntensity}
  n179 = Divide(
   A=@n171.Result
   B=LinearInterpolate(
    A=ScalarParameter() {defaultValue=325, group=Kiwi_Goo, parameterName=Kiwi_Wave_WarpNoise1Tilerate}
    B=ScalarParameter() {defaultValue=175, group=SinkingGradient, parameterName=Kiwi_SinkGrad_GooWarpNoise1Tilerate, sortPriority=50}
    Alpha=@n174
   )
  )
  n180 = Multiply(A=Time(), B=ScalarParameter() {defaultValue=0.075, group=Kiwi_Goo, parameterName=Kiwi_Wave_WarpNoise1Speed})
  n181 = ScalarParameter() {defaultValue=450, group=Kiwi_Goo, parameterName=Kiwi_Wave_WarpNoise2TileRate}
  n182 = ScalarParameter() {defaultValue=650, group=SinkingGradient, parameterName=Kiwi_SinkGrad_GooWarpNoise2Tilerate, sortPriority=50}
  n183 = Divide(A=@n171.Result, B=LinearInterpolate(A=LinearInterpolate(A=@n181, B=@n182, Alpha=@n174), B=@n182, Alpha=@n175))
  n184 = Multiply(A=Time(), B=ScalarParameter() {defaultValue=0.1, group=Kiwi_Goo, parameterName=Kiwi_Wave_WarpNoise2Speed})
  n185 = Multiply(
   A=TextureSample!(UVs=Panner(Coordinate=@n179, Time=@n180) {speedX=-1, speedY=0.5}, Tex, Apply View MipBias) {samplerType=SAMPLERTYPE_Grayscale, texture=&LowResBlurredNoise}.R
   B=TextureSample(UVs=Panner(Coordinate=@n183, Time=@n184) {speedX=1, speedY=-0.5}) {samplerType=SAMPLERTYPE_Grayscale, texture=&LowResBlurredNoise}.R
  )
  n186 = TextureSample(UVs=Add(A=Panner(Coordinate=@n176, Time=@n177) {speedX=-1, speedY=0.5}, B=Multiply(A=@n178, B=@n185))) {samplerType=SAMPLERTYPE_Grayscale, texture=&LowResBlurredNoise}
  Multiply(
   A=Power(Base=@n186.R, Exp=ScalarParameter() {defaultValue=3, parameterName=dasf34t_8})
   B=ScalarParameter() {defaultValue=12, parameterName=dasf34t_6}
  )
  NamedRerouteDeclaration(None=ScalarParameter() {group=Switches, parameterName=Kiwi_Purple, sortPriority=0}) {name=Kiwi_Purple, nodeColor=<rgba 1 0 0.9036 1>}
  NamedRerouteDeclaration(None=ScalarParameter() {desc="0 = Previous color 1 = Purple", group=Switches, parameterName=Kiwi_Purple_Angry, sortPriority=1}) {name=Kiwi_Purple_Angry, nodeColor=<rgba 0.5042 0 1 1>}
  NamedRerouteDeclaration(None=@n171.Result) {name=WorldAlignedUVs, nodeColor=<rgba 0 1 0.9063 1>}
  NamedRerouteDeclaration(None=@n108) {name=Event_RunesOn, nodeColor=<rgba 0.5234 1 0 1>}
  n187 = MaterialFunctionCall(UseLocalSpace=StaticBool() {value=true}) {materialFunction=&MF_WorldAlignedUVs_HardEdge}
  n188 = ScalarParameter() {defaultValue=200, group=pool, parameterName=Kiwi_SinkGrad_GooWarpTilerate Pool, sortPriority=50}
  n189 = NamedRerouteUsage()
  n190 = LinearInterpolate(A=ScalarParameter() {defaultValue=500, group=pool, parameterName=Kiwi_Wave_WarpTilerate Pool}, B=@n188, Alpha=@n189)
  n191 = NamedRerouteUsage()
  n192 = Panner(
   Coordinate=Divide(A=@n187.Result, B=LinearInterpolate(A=@n190, B=@n188, Alpha=@n191))
   Time=Multiply(A=Time(), B=ScalarParameter() {defaultValue=0.4, group=pool, parameterName=Kiwi_Wave_WarpSpeed Pool})
  ) {speedX=-1, speedY=0.5}
  n193 = Divide(
   A=@n187.Result
   B=LinearInterpolate(
    A=ScalarParameter() {defaultValue=325, group=pool, parameterName=Kiwi_Wave_WarpNoise1Tilerate Pool}
    B=ScalarParameter() {defaultValue=175, group=pool, parameterName=Kiwi_SinkGrad_GooWarpNoise1Tilerate Pool, sortPriority=50}
    Alpha=@n189
   )
  )
  n194 = Multiply(A=Time(), B=ScalarParameter() {defaultValue=0.075, group=pool, parameterName=Kiwi_Wave_WarpNoise1Speed Pool})
  n195 = TextureObjectParameter() {
   group=pool
   parameterName=Pool Gradient texture
   samplerType=SAMPLERTYPE_Grayscale
   texture=&LowResBlurredNoise
  }
  n196 = ScalarParameter() {defaultValue=450, group=pool, parameterName=Kiwi_Wave_WarpNoise2TileRate Pool}
  n197 = ScalarParameter() {defaultValue=650, group=pool, parameterName=Kiwi_SinkGrad_GooWarpNoise2TileratePool, sortPriority=50}
  n198 = Divide(A=@n187.Result, B=LinearInterpolate(A=LinearInterpolate(A=@n196, B=@n197, Alpha=@n189), B=@n197, Alpha=@n191))
  n199 = Multiply(A=Time(), B=ScalarParameter() {defaultValue=0.1, group=pool, parameterName=Kiwi_Wave_WarpNoise2Speed Pool})
  n200 = Multiply(
   A=TextureSample(UVs=Panner(Coordinate=@n193, Time=@n194) {speedX=-1, speedY=0.5}, Tex=@n195) {samplerType=SAMPLERTYPE_Grayscale, texture=&LowResBlurredNoise}.R
   B=TextureSample(UVs=Panner(Coordinate=@n198, Time=@n199) {speedX=1, speedY=-0.5}, Tex=@n195) {samplerType=SAMPLERTYPE_Grayscale, texture=&LowResBlurredNoise}.R
  )
  n201 = Multiply(A=ScalarParameter() {defaultValue=1.5, group=pool, parameterName=Kiwi_Wave_WarpNoiseIntensity Pool}, B=@n200)
  n202 = Multiply(
   A=TextureSample(UVs=Add(A=@n192, B=@n201), Tex=@n195) {samplerType=SAMPLERTYPE_Grayscale, texture=&LowResBlurredNoise}.R
   B=@n200
  )
  NamedRerouteDeclaration(None=Saturate(None=Add(A=@n202, B=@n200))) {name=WarpNoise_Pool, nodeColor=<rgba 0 1 0.25 1>}
  n203 = Add(A=@n1.XYZ, B=@n2)
  n204 = Constant() {r=-1}
  n205 = Saturate(None=Multiply(
    A=Multiply(A=@n203, B=@n204)
    B=ScalarParameter() {defaultValue=0.0009, group=SinkingGradient, parameterName=Kiwi_SinkGrad_BlueSize, sortPriority=50}
   ))
  n206 = MaterialFunctionCall(
   Transition Phase=Power(
    Base=ComponentMask(None=@n205) {b=true}
    Exp=ScalarParameter() {defaultValue=15, group=SinkingGradient, parameterName=Kiwi_SinkGrad_BluePower, sortPriority=50}
   )
   Height Texture=NamedRerouteUsage()
   Contrast=ScalarParameter() {defaultValue=25, group=SinkingGradient, parameterName=Kiwi_SinkGrad_BlueContrast, sortPriority=50}
  ) {materialFunction=&HeightLerp}
  NamedRerouteDeclaration(None=@n206.Alpha) {name=SinkingGradient, nodeColor=<rgba 0.0547 1 0 1>}
  NamedRerouteDeclaration(None=ScalarParameter() {desc="0 = uses defualt 1 = uses sinking gradient with goo material", group=Switches, parameterName=Kiwi_Goo, sortPriority=0}) {name=Kiwi_Goo, nodeColor=<rgba 0 0.0395 1 1>}
  NamedRerouteDeclaration(None=LinearInterpolate(
    A=@n122
    B=ScalarParameter() {defaultValue=0.5, parameterName=Kiwi_SimpleSpecular}
    Alpha=ScalarParameter() {desc="0=default Material atts 1=Simplified Atts", parameterName=Kiwi_SimpleSpecular_On}
   )) {name=Specular, nodeColor=<rgba 0 1 0.7891 1>}
  n207 = Multiply(A=@n186.R, B=@n185)
  NamedRerouteDeclaration(None=Saturate(None=Add(A=@n207, B=@n185))) {name=WarpNoise, nodeColor=<rgba 0 1 0.25 1>}
  n208 = TextureCoordinate() {coordinateIndex=2}
  n209 = Constant2Vector() {g=0.5, r=0.5}
  n210 = MaterialFunctionCall(Float2=Abs(None=Subtract(A=@n208, B=@n209))) {materialFunction=&BreakOutFloat2Components}
  NamedRerouteDeclaration(None=Max(A=@n210.R, B=@n210.R)) {name=BoxMask, nodeColor=<rgba 0.2422 1 0 1>}
  NamedRerouteDeclaration(None=ScalarParameter() {desc="0 = Previous color 1 = White", group=Switches, parameterName=Kiwi_White, sortPriority=3}) {name=Kiwi_White, nodeColor=<rgba 1 1 1 1>}
  NamedRerouteDeclaration(None=LinearInterpolate(
    A=@n125
    B=ScalarParameter() {defaultValue=0.5, parameterName=Kiwi_SimpleRoughness}
    Alpha=ScalarParameter() {desc="0=default Material atts 1=Simplified Atts", parameterName=Kiwi_SimpleRoughness_On}
   )) {name=Roughness, nodeColor=<rgba 1 0.2109 0 1>}
  NamedRerouteDeclaration(None=@n155) {name=BaseColor, nodeColor=<rgba 1 0 0.4687 1>}
  n211 = Power(Base=@n206.Lerp Alpha No Contrast) {constExponent=5}
  n212 = Multiply(A=Add(A=NamedRerouteUsage()) {constB=0.2}) {constB=0.8}
  n213 = NamedRerouteUsage()
  n214 = Multiply(A=LinearInterpolate(B=@n212, Alpha=@n213)) {constB=0.2}
  n215 = Multiply(A=NamedRerouteUsage()) {constB=0.1}
  n216 = LinearInterpolate(A=@n5, Alpha=@n6) {constB=2}
  n217 = TextureSample(UVs=Divide(A=NamedRerouteUsage()) {constB=150}) {samplerSource=SSM_Wrap_WorldGroupSettings, samplerType=SAMPLERTYPE_Grayscale, texture=&LowResBlurredNoise}
  n218 = Power(Base=Multiply(A=@n215, B=@n217.R)) {constExponent=1.5}
  n219 = NamedRerouteUsage()
  n220 = Add(A=LinearInterpolate(B=@n215, Alpha=@n213), B=Add(A=Saturate(None=@n216), B=LinearInterpolate(B=@n218, Alpha=@n219)))
  n221 = Power(Base=ComponentMask(None=Add(A=@n214, B=@n220)) {b=true}, Exp=@n117)
  n222 = VectorParameter() {group=SinkingGradient, parameterName=Kiwi_SinkGrad_ColorWhite, defaultValue=<rgba 1.25 1.25 1.25 1>}
  n223 = LinearInterpolate(
   A=VectorParameter() {group=SinkingGradient, parameterName=Kiwi_SinkGrad_Color, defaultValue=<rgba 0.8957 0 1.25 1>}.RGB
   B=@n222.RGB
   Alpha=NamedRerouteUsage()
  )
  n224 = NamedRerouteUsage()
  n225 = NamedRerouteUsage()
  n226 = Multiply(A=Multiply(A=NamedRerouteUsage()) {constB=1.25}, B=Add(A=NamedRerouteUsage(), B=Add(A=@n225, B=@n225)))
  n227 = Add(
   A=VectorParameter() {group=SinkingGradient, parameterName=Kiwi_SinkGrad_ColorBlue, defaultValue=<rgba 0 0.5694 1.25 1>}.RGB
   B=LinearInterpolate(B=Multiply(A=@n226) {constB=0.4}, Alpha=NamedRerouteUsage())
  )
  n228 = LinearInterpolate(
   A=LinearInterpolate(A=LinearInterpolate(A=@n223, B=@n222.RGB, Alpha=@n224), B=@n227, Alpha=NamedRerouteUsage())
   B=@n227
   Alpha=NamedRerouteUsage()
  )
  n229 = Multiply(A=LinearInterpolate(A=Add(A=@n211, B=@n221), B=Power(Base=@n206.Lerp Alpha No Contrast), Alpha=NamedRerouteUsage()), B=@n228)
  n230 = ScalarParameter() {defaultValue=1, group=SinkingGradient, parameterName=Sinking Gradient Multiplier}
  n231 = StaticSwitchParameter(True=Add(A=Multiply(A=@n229, B=@n230), B=@n116), False=@n116) {defaultValue=true, group=SinkingGradient, parameterName=Use Sinking Gradient, sortPriority=0}
  n232 = Add(
   A=WorldPosition().XYZ
   B=Add(
    A=ScalarParameter() {defaultValue=-99876.5469, group=pool, parameterName=Kiwi_SinkGrad_Offset Pool Offset}
    B=ScalarParameter() {group=pool, parameterName=Kiwi_SinkGrad_Offset Pool}
   )
  )
  n233 = Saturate(None=Multiply(A=Multiply(A=@n232) {constB=-1}, B=ScalarParameter() {defaultValue=0.0009, group=pool, parameterName=Kiwi_SinkGrad_Size pool}))
  n234 = MaterialFunctionCall(
   Transition Phase=Power(
    Base=ComponentMask(None=@n233) {b=true}
    Exp=ScalarParameter() {defaultValue=15, group=pool, parameterName=Kiwi_SinkGrad_BluePower Pool, sortPriority=50}
   )
   Height Texture=NamedRerouteUsage()
   Contrast=ScalarParameter() {defaultValue=25, group=pool, parameterName=Kiwi_SinkGrad_BlueContrastpool, sortPriority=50}
  ) {materialFunction=&HeightLerp}
  n235 = ScalarParameter() {defaultValue=2, parameterName=Kiwi_SinkGrad_BlueContrast pool Exp}
  n236 = Multiply(
   A=Multiply(A=Power(Base=@n234.Alpha, Exp=@n235), B=ScalarParameter() {defaultValue=1, parameterName=Kiwi_SinkGrad_BlueContrast pool Multiplier})
   B=ScalarParameter() {defaultValue=1, parameterName=Kiwi_SinkGrad_BlueContrast pool Multiplier 2}
  )
  n237 = Desaturation(
   None=Multiply(A=@n228, B=NamedRerouteDeclaration(None=@n236) {name=SinkingGradientPool, nodeColor=<rgba 0.0547 1 0 1>})
   Fraction=ScalarParameter() {group=pool, parameterName=Pool Desaturation Amount}
  )
  NamedRerouteDeclaration(None=StaticSwitchParameter(True=Add(A=@n231, B=@n237), False=@n231) {group=pool, parameterName=Use Pool Gradient, sortPriority=0}) {name=Emissive, nodeColor=<rgba 0.5 0 1 1>}
  n238 = TextureSample(UVs=@n142) {samplerSource=SSM_Wrap_WorldGroupSettings, samplerType=SAMPLERTYPE_Masks, texture=&T_CubeNoise}
  NamedRerouteDeclaration(None=@n238.G) {name=Mask_GreenChannel, nodeColor=<rgba 0 1 0.2266 1>}
  NamedRerouteDeclaration(None=@n185) {name=Largenoise, nodeColor=<rgba 1 0.8672 0 1>}
  NamedRerouteDeclaration(None=Multiply(A=@n54.RGB, B=@n103)) {name=Runes, nodeColor=<rgba 1 0.7734 0 1>}
  NamedRerouteDeclaration(None=@n221) {name=SinkGradientMask, nodeColor=<rgba 0 1 0.8359 1>}
  NamedRerouteDeclaration(None=@n44) {name=NoiseMask, nodeColor=<rgba 0.7344 0 1 1>}
  NamedRerouteDeclaration(None=@n140) {name=Sparkles, nodeColor=<rgba 1 0 0.7031 1>}
  NamedRerouteDeclaration(None=@n50) {name=EdgeMask, nodeColor=<rgba 1 0 0.7734 1>}
  n239 = ScalarParameter() {desc="0 = Previous color 1 = Tranisiton color to white", group=Switches, parameterName=Kiwi_White_Transition, sortPriority=2}
  NamedRerouteDeclaration(None=@n239) {name=Kiwi_White_Transition, nodeColor=<rgba 0.2024 0.2006 0.1891 1>}
  NamedRerouteDeclaration(None=@n148) {name=BoxMaskEdge, nodeColor=<rgba 0.125 1 0 1>}
  NamedRerouteDeclaration(None=@n28) {name=TextureMask, nodeColor=<rgba 1 0.0469 0 1>}
  NamedRerouteDeclaration(None=@n170) {name=Hand Mask, nodeColor=<rgba 1 0 0.9609 1>}
  NamedRerouteDeclaration(None=@n212) {name=LargeNoiseModded, nodeColor=<rgba 0.875 0 1 1>}
  NamedRerouteDeclaration(None=ScalarParameter() {desc="0 = Previous color 1 = blue color", group=Switches, parameterName=Kiwi_Blue, sortPriority=5}) {name=Kiwi_Blue, nodeColor=<rgba 0 0.0395 1 1>}
  NamedRerouteDeclaration(None=ScalarParameter() {desc="0 = Previous color 1 = Tranisiton color to Inert Gray", group=Switches, parameterName=Kiwi_Inert_Gray, sortPriority=4}) {name=Kiwi_Inert, nodeColor=<rgba 0 0.7652 1 1>}
  NamedRerouteDeclaration(None=Power(Base=@n143.R, Exp=Constant() {r=3})) {name=ShadowMask_RedChannel, nodeColor=<rgba 1 0.0975 0 1>}
  NamedRerouteDeclaration(None=@n143.G) {name=ShadowMask_GreenChannel, nodeColor=<rgba 0.0165 1 0 1>}
  NamedRerouteDeclaration(None=@n238.R) {name=Mask_RedChannel, nodeColor=<rgba 1 0 0.7031 1>}
  NamedRerouteDeclaration(None=@n38."Rounded Linear Sine ") {name=SineWave, nodeColor=<rgba 0.7344 1 0 1>}
  NamedRerouteUsage()
  NamedRerouteUsage()
  NamedRerouteUsage()
  SetMaterialAttributes!(
   MaterialAttributes
   Base Color=NamedRerouteUsage()
   Metallic=ScalarParameter() {defaultValue=0.2, parameterName=Metallic}
   Specular=NamedRerouteUsage()
   Roughness=NamedRerouteUsage()
   Emissive Color=NamedRerouteUsage()
   Normal=Constant3Vector() {constant=<rgba 0 0 1 0>}
  ) {attributeSetTypes [
    #1
    #2
    #3
    #4
    #5
    #6
   ]}
 ]
}
